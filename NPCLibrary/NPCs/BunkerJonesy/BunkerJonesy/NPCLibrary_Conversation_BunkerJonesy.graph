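# Conversation database for the Bunker Jonesy NPC, defining his greeting and the options that follow it.
# The entry point leads to one speech node with a default chore-wheel line and contextual overrides for first-ever meetings or specific player character cosmetics.
# The cosmetic-specific lines carry raised priority and weight, while the first-meeting line takes its priority from a curve row.
# The speech node then links out to the shared NPC library quests and basic services conversations.
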ConversationDatabase /NPCLibrary/NPCs/BunkerJonesy/BunkerJonesy/NPCLibrary_Conversation_BunkerJonesy
entry_points [{tag=Conversation.Entry.BunkerJonesy, destinations [#1]}]
nodes [
 ConversationEntryPointNode_7 {
  id=#1
  class=ConversationEntryPointNode
  outputs [#2]
  properties {entryTag {tagName=Conversation.Entry.BunkerJonesy}}
 }
 Speech_C_4 {
  id=#2
  class=Speech_C
  outputs [#3, #4]
  properties {
   nodeName=""
   generalConfig {
    defaultMessage=You wanna help me? Pick a task off the chore wheel.
    contextualMessages [
     {
      message="No more conspiracies for me, buddy. I'm retired."
      contextRequirements [{requirement {class=FortControllerRequirement_IsFirstEverConversationWithNPC}}]
      priority {curve {rowName=Tandem.Speech.Priority.FirstEverInteraction}}
     }
     {
      message=Why do the Joneses listen to you? Is it the suit?
      contextRequirements [{requirement {
         class=FortControllerRequirement_HasCID
         properties {allowedCIds [&CID_989_Athena_Commando_M_ProgressiveJonesy]}
        }}]
      priority {value=10, curve {curveTable=null, rowName=None}}
      weight {value=10}
     }
     {
      message="No! I'm out! Let me enjoy my retirement in peace!"
      contextRequirements [{requirement {
         class=FortControllerRequirement_HasCID
         properties {allowedCIds [/BRCosmetics/Athena/Items/Cosmetics/Characters/CID_409_Athena_Commando_M_BunkerMan]}
        }}]
      priority {value=10, curve {curveTable=null, rowName=None}}
      weight {value=10}
     }
     {
      message="Buddy, you NEED to learn to shower."
      contextRequirements [{requirement {
         class=FortControllerRequirement_HasCID
         properties {allowedCIds [/BRCosmetics/Athena/Items/Cosmetics/Characters/CID_836_Athena_Commando_M_JonesyFlare]}
        }}]
      priority {value=10, curve {curveTable=null, rowName=None}}
      weight {value=10}
     }
     {
      message=Did you use all my red yarn for... KNITTING??
      contextRequirements [{requirement {
         class=FortControllerRequirement_HasCID
         properties {allowedCIds [/BRCosmetics/Athena/Items/Cosmetics/Characters/CID_798_Athena_Commando_M_JonesyVagabond]}
        }}]
      priority {value=10, curve {curveTable=null, rowName=None}}
      weight {value=10}
     }
     {
      message=You are NOT above cleaning the drains.
      contextRequirements [{requirement {
         class=FortControllerRequirement_HasCID
         properties {allowedCIds [&CID_883_Athena_Commando_M_ChOneJonesy]}
        }}]
      priority {value=10, curve {curveTable=null, rowName=None}}
      weight {value=10}
     }
     {
      message="You're back! How do you feel about vaccuuming?"
      contextRequirements [{requirement {
         class=FortControllerRequirement_HasCID
         properties {allowedCIds [/BRCosmetics/Athena/Items/Cosmetics/Characters/CID_A_431_Athena_Commando_M_JonesyOrange]}
        }}]
      priority {value=10, curve {curveTable=null, rowName=None}}
      weight {value=10}
     }
     {
      message=Can your sick tricks get you onto the roof? It needs fixing.
      contextRequirements [{requirement {class=FortControllerRequirement_HasCID, properties {allowedCIds [&Character_ChillCat]}}}]
      priority {value=10, curve {curveTable=null, rowName=None}}
      weight {value=10}
     }
     {
      message="You need anything, ask Jones, Jones, Jones or Jones... or maybe Jones."
      contextRequirements [{requirement {class=FortControllerRequirement_HasCID, properties {allowedCIds [&Character_Bites]}}}]
      priority {value=10, curve {curveTable=null, rowName=None}}
      weight {value=10}
     }
     {
      message="If you need a new home, you're welcome at Fort Jonesy."
      contextRequirements [{requirement {class=FortControllerRequirement_HasCID, properties {allowedCIds [&Character_PinkSpike]}}}]
      priority {value=10, curve {curveTable=null, rowName=None}}
      weight {value=10}
     }
     {
      message="I looked up your old rivals. Funny, they're all... 'missing'."
      contextRequirements [{requirement {class=FortControllerRequirement_HasCID, properties {allowedCIds [&Character_BadBear]}}}]
      priority {value=10, curve {curveTable=null, rowName=None}}
      weight {value=10}
     }
     {
      message="I'm watching you. "
      contextRequirements [{requirement {class=FortControllerRequirement_HasCID, properties {allowedCIds [&Character_TheHerald]}}}]
      priority {value=10, curve {curveTable=null, rowName=None}}
      weight {value=10}
     }
     {
      message=A spider person. Am I the only one seeing this?
      contextRequirements [{requirement {class=FortControllerRequirement_HasCID, properties {allowedCIds [&Character_RoseDust]}}}]
      priority {value=10, curve {curveTable=null, rowName=None}}
      weight {value=10}
     }
    ]
   }
  }
 }
 ConversationLinkNode_0 {id=#4, class=ConversationLinkNode, properties {remoteEntryTag {tagName=Conversation.Entry.BasicServices}}}
 ConversationLinkNode_1 {id=#3, class=ConversationLinkNode, properties {remoteEntryTag {tagName=Conversation.Entry.NPCLibrary.Quests}}}
]
speakers [{}, {participant=Conversation.Participant.NPC}]
